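# Parent material for Season 26 Super Level character skins, layering animated styles over the standard character PreFX/PostFX functions.
# An FX mask splits primary, secondary, tertiary, emissive and quaternary regions, each lerped into base color, metallic, roughness, specular, normal or emissive.
# Primary uses T1 stepped lines, T2 SDF lines or T3 matrix lines, scrolled on character-stabilized screen-space UVs.
# Static switches add toon shading, galaxy hair, Lazarus variants and debug views.
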Material /Game/Characters/Player/Female/Medium/Bodies/F_MED_PizzaParty/Skins/SuperLevel/S26_SL_Parent_Materials/M_S26_SL_PARENT
graph {
 properties {
  bUseMaterialAttributes=true
  bUsedWithClothing=true
  bUsedWithMeshDeformer=true
  bUsedWithMorphTargets=true
  bUsedWithNanite=true
  bUsedWithSkeletalMesh=true
  pixelDepthOffsetMode=PDOM_Legacy
  shadingModel=MSM_FromMaterialExpression
  subsurfaceProfile=&SS_HeroSkin_02
  usedShadingModels=Default Lit | Subsurface | Clear Coat | Subsurface Profile | Cloth
 }
 nodes [
  n1 = ScalarParameter() {defaultValue=0.2, group=*S26_SL_T1 Pattern, parameterName=T1_LineThickness}
  MaterialExpressionReroute_3 = Reroute(None=@n1)
  n2 = Panner(
   Coordinate=NamedRerouteUsage()
   Time=NamedRerouteUsage()
   Speed=AppendVector(
    A=ScalarParameter() {defaultValue=-0.04, group=*S26SL_Primary, parameterName=Layer2SpeedX}
    B=ScalarParameter() {defaultValue=0.02, group=*S26SL_Primary, parameterName=Layer2SpeedY}
   )
  )
  n3 = TextureSample(UVs=@n2, Tex=NamedRerouteUsage()) {
   samplerType=SAMPLERTYPE_LinearColor
   texture=/Game/Characters/Player/Female/Medium/Bodies/F_MED_PizzaParty/Skins/SuperLevel/Textures/T_SL26_Pattern_T2
  }
  n4 = Time()
  n5 = ScalarParameter() {defaultValue=0.15, group=*S26_SL_T1 Pattern, parameterName=Anim_PatternSineFreq}
  n6 = Divide(A=Sine(None=Multiply(A=@n4, B=@n5) {constB=0.4}) {period=2.5}) {constB=3}
  MaterialExpressionReroute_8 = Reroute(None=Clamp(None=Add(A=@n6) {constB=0.5}, Min=Add(A=@n1) {constB=0.05}) {minDefault=0.35})
  n7 = TextureSample(UVs=@n2, Tex=NamedRerouteUsage()) {
   samplerType=SAMPLERTYPE_LinearColor
   texture=&T_SL26_Pattern_T2
  }
  n8 = NamedRerouteUsage()
  Composite!(
   lineThickness=@MaterialExpressionReroute_3
   pattern=@n3.G
   time=@MaterialExpressionReroute_8
   mask=StaticSwitchParameter(True=@n7.G, False=@n3.G) {group=*S26SL_Primary, parameterName=ExtraMaskTexture}
   linesAmount=@n8
  )
  n9 = Panner(
   Coordinate=NamedRerouteUsage()
   Time=NamedRerouteUsage()
   Speed=AppendVector(
    A=ScalarParameter() {group=*S26SL_Primary, parameterName=Layer3SpeedX}
    B=ScalarParameter() {defaultValue=0.02, group=*S26SL_Primary, parameterName=Layer3SpeedY}
   )
  ) {speedX=-0.5, speedY=-0.05}
  n10 = TextureSample(UVs=@n9, Tex=NamedRerouteUsage()) {
   samplerType=SAMPLERTYPE_LinearColor
   texture=&T_SL26_Pattern_T2
  }
  n11 = TextureSample(UVs=@n9, Tex=NamedRerouteUsage()) {
   samplerType=SAMPLERTYPE_LinearColor
   texture=&T_SL26_Pattern_T2
  }
  n12 = StaticSwitchParameter(True=@n11.B, False=@n10.B) {group=*S26SL_Primary, parameterName=ExtraMaskTexture}
  n13 = NamedRerouteUsage()
  Composite(lineThickness=@MaterialExpressionReroute_3, pattern=@n10.B, time=@MaterialExpressionReroute_8, mask=@n12, linesAmount=@n13)
  n14 = Panner(Coordinate=NamedRerouteUsage(), Time=NamedRerouteUsage(), Speed=NamedRerouteUsage()) {speedX=-0.02, speedY=0.02}
  n15 = TextureSample(UVs=@n14, Tex=NamedRerouteUsage()) {
   samplerType=SAMPLERTYPE_LinearColor
   texture=&T_SL26_Pattern_T2
  }
  n16 = TextureSample(UVs=@n14, Tex=NamedRerouteUsage()) {
   samplerType=SAMPLERTYPE_LinearColor
   texture=&T_SL26_Pattern_T2
  }
  n17 = NamedRerouteUsage()
  Composite(
   lineThickness=@MaterialExpressionReroute_3
   pattern=@n15.R
   time=@MaterialExpressionReroute_8
   mask=StaticSwitchParameter(True=@n16.R, False=Constant() {r=1}) {group=*S26SL_Primary, parameterName=ExtraMaskTexture}
   linesAmount=@n17
  )
  MaterialExpressionReroute_2 = Reroute(None=@n17)
  n18 = NamedRerouteUsage()
  MaterialExpressionReroute_1 = Reroute(None=@n18)
  n19 = Multiply(A=@MaterialExpressionReroute_1.pattern, B=Multiply(A=@MaterialExpressionReroute_2.linesAmount) {constB=0.5})
  Frac(None=@n19)
  n20 = MaterialFunctionCall() {materialFunction=&MF_CharacterParent_2023_PreFX}
  MaterialFunctionCall(
   Material=NamedRerouteUsage()
   M=@n20.M
   Pre Skinned Local Position=NamedRerouteDeclaration(None=@n20.Pre skinned Local Position) {name=PreSkinLocalPos_IN, nodeColor=<rgba 0 1 0.8594 1>}
   Distance Blend=@n20.M
   Fresnel=@n20.M
  ) {materialFunction=&MF_CharacterParent_2023_PostFX}
  Multiply(A=@MaterialExpressionReroute_2.time)
  NamedRerouteDeclaration(None=@n20.Result) {name=MA_IN, nodeColor=<rgba 0.7344 0 1 1>}
  n21 = Constant()
  n22 = StaticSwitchParameter(
   True=StaticSwitchParameter(True=NamedRerouteUsage(), False=@n21) {defaultValue=true, group=*S26_SL_Debug, parameterName=UseHairEmissive}
   False=@n21
  ) {group=*S26_SL_Hair, parameterName=UseHair}
  n23 = Constant()
  n24 = SetMaterialAttributes!(MaterialAttributes, Emissive Color=NamedRerouteUsage(), Base Color=@n23, Metallic=@n23, Roughness=Constant() {r=1}) {attributeSetTypes [
    #1
    #2
    #3
    #4
   ]}
  n25 = NamedRerouteUsage()
  n26 = MaterialFunctionCall!(
   Material=@n25
   LitDiffuseOverride
   SSC_AlphaOverride
   Tint Mask
   DF InkLine Visibility Mask
   DF Inkline Injection
   DF Inkline Color Override
   DF Inkline Color Override Mask
   CustomUVs
  ) {materialFunction=/Game/Packages/Fortress_SharedMaterials/Anime_Master_Material/MF_AnimeShading}
  n27 = MaterialFunctionCall!(
   Material=MaterialFunctionCall!(In=@n26.Result) {materialFunction=/Game/Packages/Fortress_SharedMaterials/Anime_Master_Material/MF_AnimatedVisor}.Result
   SkinFX_Mask
  ) {materialFunction=/Game/Global/MaterialFunctions/MF_ClothSimWPO_v2}
  n28 = GetMaterialAttributes(None=StaticSwitchParameter(
    True=MaterialFunctionCall!(MA=@n27.Result, Shading=@n26.Shading, InkLine Mask=@n26.Shading, SSC=@n26.Shading, SkinFX=@n26.Shading) {materialFunction=/Game/Packages/Fortress_SharedMaterials/Anime_Master_Material/MF_AnimeSkinFX}.MA
    False=@n25
   ) {group=*S26_SL, parameterName=isToon}) {attributeGetTypes [
    #2
    #4
    #3
    #5
    #1
    #6
    #7
   ]}
  n29 = SetMaterialAttributes!(
   MaterialAttributes=@n28.MaterialAttributes
   Base Color=NamedRerouteUsage()
   Roughness=NamedRerouteUsage()
   Metallic=NamedRerouteUsage()
   Specular=NamedRerouteUsage()
   Emissive Color=NamedRerouteUsage()
   Normal=NamedRerouteUsage()
  ) {attributeSetTypes [
    #2
    #4
    #3
    #5
    #1
    #6
   ]}
  NamedRerouteDeclaration(None=StaticSwitchParameter(
    True=SetMaterialAttributes!(MaterialAttributes, Base Color=@n22) {attributeSetTypes [#2]}
    False=StaticSwitchParameter(True=@n24, False=@n29) {group=*S26_SL, parameterName=UseDebug_SL_Mask}
   ) {group=*S26_SL_Debug, parameterName=UseDebug_Emissive}) {name=MA_OUT, nodeColor=<rgba 0 1 0.2031 1>}
  n30 = VectorParameter() {group=*S26SL_Secondary, parameterName=SecondaryColor, sortPriority=1, defaultValue=<rgba 1 0.2658 0 1>}
  NamedRerouteDeclaration(None=@n30.RGB) {name=SecondaryColor, nodeColor=<rgba 0.0064 0 1 1>}
  n31 = VectorParameter() {group=*S26SL_Tertiary, parameterName=TertiaryColor, sortPriority=2, defaultValue=<rgba 0.4233 0.2502 0.0545 1>}
  NamedRerouteDeclaration(None=@n31.RGB) {name=TertiaryColor, nodeColor=<rgba 0.0225 1 0 1>}
  NamedRerouteDeclaration(None=Multiply(A=NamedRerouteUsage(), B=ScalarParameter() {defaultValue=1, parameterName=Quaternary_RoughnessIntensity})) {name=Quaternary_Roughness, nodeColor=<rgba 0 0.9766 1 1>}
  NamedRerouteDeclaration(None=NamedRerouteUsage()) {name=BaseColor_OUT, nodeColor=<rgba 0.3505 1 0 1>}
  n32 = StaticSwitchParameter(True=NamedRerouteUsage(), False=LinearInterpolate(A=NamedRerouteUsage(), B=NamedRerouteUsage(), Alpha=NamedRerouteUsage())) {group=*S26_SL, parameterName=isLazarusWingHead}
  n33 = NamedRerouteUsage()
  n34 = NamedRerouteUsage()
  n35 = NamedRerouteUsage()
  n36 = NamedRerouteUsage()
  n37 = LinearInterpolate(
   A=StaticSwitchParameter(True=LinearInterpolate(A=@n32, B=@n33, Alpha=@n34), False=LinearInterpolate(A=@n32, B=@n35, Alpha=@n36)) {group=*S26_SL_Primary_Lazarus, parameterName=isLazarusLensStyleMask}
   B=NamedRerouteUsage()
   Alpha=NamedRerouteUsage()
  )
  n38 = LinearInterpolate(A=@n37, B=NamedRerouteUsage(), Alpha=NamedRerouteUsage()) {constAlpha=1}
  NamedRerouteDeclaration(None=LinearInterpolate(A=StaticSwitchParameter(True=@n38, False=@n37) {group=*S26_SL_Hair, parameterName=UseHair}, B=NamedRerouteUsage(), Alpha=NamedRerouteUsage())) {name=SL_BaseColor, nodeColor=<rgba 0 0.1328 1 1>}
  n39 = LinearInterpolate(A=NamedRerouteUsage(), B=NamedRerouteUsage(), Alpha=NamedRerouteUsage())
  n40 = NamedRerouteUsage()
  n41 = NamedRerouteUsage()
  n42 = Multiply(A=NamedRerouteUsage()) {constB=0.2}
  n43 = Multiply(A=NamedRerouteUsage()) {constB=2}
  n44 = Saturate(None=Add(A=OneMinus(None=@n43), B=NamedRerouteUsage()))
  NamedRerouteDeclaration(None=StaticSwitchParameter(True=LinearInterpolate(A=@n39, B=@n40, Alpha=@n41), False=LinearInterpolate(A=@n42, B=@n39, Alpha=@n44)) {group=*S26_SL, parameterName=UseT3}) {name=Primary_BaseColor, nodeColor=<rgba 1 0 0.1875 1>}
  n45 = GetMaterialAttributes(None=MaterialFunctionCall!(SequinColorInner, Fresnel=NamedRerouteUsage(), inMat, Mask=Constant() {r=1}, SequinColorOuter) {materialFunction=/Game/Gadgets/Assets/VinderTech_Umbrella/Umbrella_Buffet/Materials/MF_BuffetSequins}.Result) {attributeGetTypes [
    #2
    #3
    #5
    #4
    #6
    #1
   ]}
  NamedRerouteDeclaration(None=StaticSwitchParameter(True=NamedRerouteUsage(), False=@n45.Base Color) {group=*S26_SL, parameterName=UseT3}) {name=Secondary_BaseColor, nodeColor=<rgba 0 0.6016 1 1>}
  NamedRerouteDeclaration(None=Fresnel() {baseReflectFraction=0, exponent=1}) {name=Fresnel, nodeColor=<rgba 0 1 0.1797 1>}
  NamedRerouteDeclaration(None=@n28.Metallic) {name=Metallic_IN, nodeColor=<rgba 0 0.3906 1 1>}
  n46 = TextureObjectParameter() {
   group=*S26_SL
   parameterName=SL_FX_Mask
   samplerType=SAMPLERTYPE_Masks
   texture=/Game/Characters/Player/Female/Medium/Bodies/F_MED_IndieBucket/Textures/T_IndieBucket_Body_TechArtFX_Masks
  }
  n47 = NamedRerouteDeclaration(None=TextureSample(Tex=@n46).RGB) {name=SL_SkinFX, nodeColor=<rgba 1 0.4922 0 1>}
  n48 = Subtract(A=@n47, B=NamedRerouteUsage())
  NamedRerouteDeclaration(None=ChannelMaskParameter(None=@n48) {group=*S26_SL, parameterName=PrimaryChannel, sortPriority=3}.R) {name=Primary_Mask, nodeColor=<rgba 1 0.0253 0 1>}
  n49 = LinearInterpolate(A=NamedRerouteUsage(), B=NamedRerouteUsage(), Alpha=NamedRerouteUsage())
  n50 = StaticSwitchParameter(
   True=LinearInterpolate(A=@n49, B=NamedRerouteUsage(), Alpha=NamedRerouteUsage())
   False=LinearInterpolate(A=@n49, B=NamedRerouteUsage(), Alpha=NamedRerouteUsage())
  ) {group=*S26_SL_Primary_Lazarus, parameterName=isLazarusLensStyleMask}
  n51 = NamedRerouteUsage()
  n52 = NamedRerouteUsage()
  n53 = NamedRerouteUsage()
  n54 = ScalarParameter() {group=*S26_SL, parameterName=NotSL_Metallic}
  n55 = ScalarParameter() {defaultValue=1, group=*S26_SL, parameterName=NotSL_Metallic_Blend}
  n56 = NamedRerouteUsage()
  n57 = LinearInterpolate(
   A=LinearInterpolate(A=LinearInterpolate(A=@n50, B=@n51, Alpha=@n52), Alpha=NamedRerouteUsage()) {constB=0}
   B=NamedRerouteDeclaration(None=LinearInterpolate(A=@n53, B=@n54, Alpha=@n55)) {name=Metallic_Blend, nodeColor=<rgba 1 0.0234 0 1>}
   Alpha=Multiply(A=NamedRerouteUsage(), B=OneMinus(None=@n56))
  )
  n58 = LinearInterpolate(A=@n57, B=NamedRerouteUsage(), Alpha=NamedRerouteUsage())
  NamedRerouteDeclaration(None=LinearInterpolate(A=StaticSwitchParameter(True=@n58, False=@n57) {group=*S26_SL_Hair, parameterName=UseHair}, B=NamedRerouteUsage(), Alpha=NamedRerouteUsage())) {name=Metallic_OUT, nodeColor=<rgba 0.0364 0.3906 0.379 1>}
  NamedRerouteDeclaration(None=@n28.Roughness) {name=Roughness_IN, nodeColor=<rgba 0.0547 0 1 1>}
  n59 = LinearInterpolate(A=NamedRerouteUsage(), B=NamedRerouteUsage(), Alpha=NamedRerouteUsage())
  n60 = NamedRerouteUsage()
  n61 = NamedRerouteUsage()
  n62 = NamedRerouteUsage()
  n63 = NamedRerouteUsage()
  n64 = LinearInterpolate(
   A=StaticSwitchParameter(True=LinearInterpolate(A=@n59, B=@n60, Alpha=@n61), False=LinearInterpolate(A=@n59, B=@n62, Alpha=@n63)) {group=*S26_SL_Primary_Lazarus, parameterName=isLazarusLensStyleMask}
   B=NamedRerouteUsage()
   Alpha=NamedRerouteUsage()
  )
  n65 = LinearInterpolate(A=@n64, B=NamedRerouteUsage(), Alpha=NamedRerouteUsage()) {constB=0}
  MaterialExpressionReroute_0 = Reroute(None=@n64)
  NamedRerouteDeclaration(None=LinearInterpolate(
    A=StaticSwitchParameter(True=@n65, False=@MaterialExpressionReroute_0) {group=*S26_SL_Hair, parameterName=UseHair}
    B=NamedRerouteUsage()
    Alpha=NamedRerouteUsage()
   )) {name=Roughness_OUT, nodeColor=<rgba 0 1 0.2969 1>}
  NamedRerouteDeclaration(None=ScalarParameter() {defaultValue=8, group=*S26SL_Primary, parameterName=PrimaryLinesAmount}) {name=Primary_LinesAmount, nodeColor=<rgba 0.5469 1 0 1>}
  NamedRerouteDeclaration(None=@n28.Specular) {name=Specular_IN, nodeColor=<rgba 0.875 1 0 1>}
  n66 = LinearInterpolate(A=NamedRerouteUsage(), B=NamedRerouteUsage(), Alpha=NamedRerouteUsage())
  n67 = LinearInterpolate(A=@n66, B=NamedRerouteUsage(), Alpha=NamedRerouteUsage())
  n68 = LinearInterpolate(A=@n66, B=NamedRerouteUsage(), Alpha=NamedRerouteUsage())
  n69 = LinearInterpolate(
   A=LinearInterpolate(
    A=StaticSwitchParameter(True=@n67, False=@n68) {group=*S26_SL_Primary_Lazarus, parameterName=isLazarusLensStyleMask}
    B=NamedRerouteUsage()
    Alpha=NamedRerouteUsage()
   )
   B=NamedRerouteUsage()
   Alpha=NamedRerouteUsage()
  ) {constB=0}
  n70 = NamedRerouteUsage()
  n71 = NamedRerouteUsage()
  NamedRerouteDeclaration(None=StaticSwitchParameter(True=LinearInterpolate(A=@n69, B=@n70, Alpha=@n71) {constB=0}, False=@n69) {group=*S26_SL_Hair, parameterName=UseHair}) {name=Specular_OUT, nodeColor=<rgba 0.3125 1 0 1>}
  NamedRerouteDeclaration(None=NamedRerouteUsage()) {name=Tertiary_BaseColor, nodeColor=<rgba 0.8281 1 0 1>}
  NamedRerouteDeclaration(None=LinearInterpolate(
    A=ScalarParameter() {defaultValue=0.75, group=*S26SL_Primary, parameterName=Primary_Metallic_Min}
    B=ScalarParameter() {defaultValue=1, group=*S26SL_Primary, parameterName=Primary_Metallic_Max}
    Alpha=NamedRerouteUsage()
   ) {constA=0.5, constB=0}) {name=Primary_Metallic, nodeColor=<rgba 1 0.5859 0 1>}
  NamedRerouteDeclaration(None=LinearInterpolate(
    A=ScalarParameter() {defaultValue=0.27, group=*S26SL_Primary, parameterName=Primary_Roughness_Min}
    B=ScalarParameter() {defaultValue=0.15, group=*S26SL_Primary, parameterName=Primary_Roughness_Max}
    Alpha=NamedRerouteUsage()
   ) {constA=0.3, constB=0}) {name=Primary_Roughness, nodeColor=<rgba 1 0.2451 0.5971 1>}
  NamedRerouteDeclaration(None=LinearInterpolate(
    A=ScalarParameter() {defaultValue=0.35, group=*S26SL_Primary, parameterName=Primary_Specular_Min}
    B=ScalarParameter() {defaultValue=0.35, group=*S26SL_Primary, parameterName=Primary_Specular_Max}
    Alpha=NamedRerouteUsage()
   ) {constB=0.5}) {name=Primary_Specular, nodeColor=<rgba 1 0.1126 0.0828 1>}
  n72 = ChannelMaskParameter(None=@n48) {group=*S26_SL, maskChannel=Blue, parameterName=SecondaryChannel, sortPriority=4, defaultValue=<rgba 0 0 1 0>}
  NamedRerouteDeclaration(None=@n72.R) {name=Secondary_Mask, nodeColor=<rgba 0 0 1 1>}
  NamedRerouteDeclaration(None=Constant()) {name=Hair_Metallic, nodeColor=<rgba 1 0.75 0 1>}
  NamedRerouteDeclaration(None=@n20.Fresnel) {name=Fresnel_IN, nodeColor=<rgba 0 0.4609 1 1>}
  NamedRerouteDeclaration(None=Constant()) {name=Hair_BaseColor, nodeColor=<rgba 0 1 0.6484 1>}
  NamedRerouteDeclaration(None=StaticSwitchParameter(True=ScalarParameter() {defaultValue=1, group=*S26SL_Secondary, parameterName=Secondary_Roughness}, False=@n45.Roughness) {group=*S26_SL, parameterName=UseT3}) {name=Secondary_Roughness, nodeColor=<rgba 0 0.5781 1 1>}
  NamedRerouteDeclaration(None=StaticSwitchParameter(True=ScalarParameter() {defaultValue=1, group=*S26SL_Secondary, parameterName=Secondary_Metallic}, False=@n45.Metallic) {group=*S26_SL, parameterName=UseT3}) {name=Secondary_Metallic, nodeColor=<rgba 0 0.2319 0.401 1>}
  NamedRerouteDeclaration(None=ScalarParameter() {defaultValue=0.2, group=*S26SL_Tertiary, parameterName=Tertiary_Roughness}) {name=Tertiary_Roughness, nodeColor=<rgba 0 1 0.2352 1>}
  n73 = NamedRerouteUsage()
  MaterialExpressionReroute_4 = Reroute(None=@n73)
  n74 = StaticSwitchParameter(True=@n45.Normal, False=@n73) {defaultValue=true, group=*S26SL_Secondary, parameterName=MFSequin_UseNormal}
  NamedRerouteDeclaration(None=StaticSwitchParameter(
    True=@MaterialExpressionReroute_4
    False=MaterialFunctionCall(BaseNormal=@n73, AdditionalNormal=@n74) {materialFunction=&BlendAngleCorrectedNormals}.Result
   ) {group=*S26_SL, parameterName=UseT3}) {name=Secondary_Normal, nodeColor=<rgba 0.4062 0 1 1>}
  n75 = NamedRerouteUsage()
  n76 = Add(A=ScalarParameter() {defaultValue=1, group=*S26SL_Secondary, parameterName=Secondary_ColorGradientHardness})
  n77 = Add(A=@n76) {constB=-1}
  n78 = Saturate(None=Add(A=Multiply(A=@n75, B=@n76), B=OneMinus(None=@n77)) {constB=0})
  n79 = NamedRerouteUsage()
  NamedRerouteDeclaration(None=StaticSwitchParameter(True=Multiply(A=@n78, B=@n79), False=@n45.Emissive Color) {group=*S26_SL, parameterName=UseT3}) {name=Secondary_Emissive, nodeColor=<rgba 0 1 0.3672 1>}
  n80 = VectorParameter() {group=*S26SL_Primary, parameterName=PrimaryInnerColor, sortPriority=0, defaultValue=<rgba 0.4323 0.2547 0.0518 1>}
  NamedRerouteDeclaration(None=@n80.RGB) {name=PrimaryInnerColor, nodeColor=<rgba 1 0.1641 0 1>}
  n81 = VectorParameter() {group=*S26SL_Secondary, parameterName=SecondaryEmissiveColor, sortPriority=2, defaultValue=<rgba 1 0.748 0.0203 1>}
  NamedRerouteDeclaration(None=@n81.RGB) {name=SecondaryEmissiveColor, nodeColor=<rgba 0.0064 0 1 1>}
  n82 = ChannelMaskParameter(None=@n48) {group=*S26_SL, maskChannel=Green, parameterName=TertiaryChannel, sortPriority=5, defaultValue=<rgba 0 1 0 0>}
  NamedRerouteDeclaration(None=@n82.R) {name=Tertiary_Mask, nodeColor=<rgba 0 1 0 1>}
  NamedRerouteDeclaration(None=Multiply(A=NamedRerouteUsage(), B=ScalarParameter() {defaultValue=1, group=*S26_SL_Quaternary, parameterName=Quaternary_MetallicIntensity})) {name=Quaternary_Metallic, nodeColor=<rgba 1 0 0.6094 1>}
  NamedRerouteDeclaration(None=ScalarParameter() {defaultValue=1, group=*S26SL_Tertiary, parameterName=Teritary_Metallic}) {name=Tertiary_Metallic, nodeColor=<rgba 0 0.2319 0.401 1>}
  n83 = MaterialFunctionCall!(BaseColor=NamedRerouteUsage(), Emissive=NamedRerouteUsage(), PreSkinLocalPos=NamedRerouteUsage()) {materialFunction=/Game/Characters/Player/Male/Medium/Bodies/M_MED_LazarusLens_Wing/Materials/MF_LazarusLens_Wings_FaceAnim}
  NamedRerouteDeclaration(None=@n83.TatooMask) {name=LazarusWing_TatooMask, nodeColor=<rgba 0.2422 1 0 1>}
  n84 = MaterialFunctionCall!(BaseColor=NamedRerouteUsage()) {materialFunction=/Game/Characters/Player/Male/Medium/Bodies/M_MED_LazarusLens_Style/Materials/MF_LazarusLensStyleMask}
  NamedRerouteDeclaration(None=@n84.Waves1) {name=LazarusLensMask_Waves1, nodeColor=<rgba 0 0.7891 1 1>}
  n85 = VectorParameter() {group=*S26SL_Primary, parameterName=PrimaryEmissiveColor, sortPriority=0, defaultValue=<rgba 1 0.9401 0.3229 1>}
  NamedRerouteDeclaration(None=@n85.RGB) {name=PrimaryEmissiveColor, nodeColor=<rgba 0 0.8828 1 1>}
  n86 = MaterialFunctionCall!(TattooMask=StaticComponentMaskParameter(None=NamedRerouteUsage()) {defaultB=true, group=*S26_SL_Primary_Lazarus, parameterName=TatooBodyTAMaskChannel}) {materialFunction=/Game/Characters/Player/Male/Medium/Bodies/M_MED_LazarusLens/Materials/MF_LazarusLens_TattooGlow}
  n87 = NamedRerouteUsage()
  n88 = LinearInterpolate(A=NamedRerouteUsage(), B=NamedRerouteUsage(), Alpha=NamedRerouteUsage())
  n89 = StaticSwitchParameter(True=LinearInterpolate(A=@n88, B=NamedRerouteUsage(), Alpha=NamedRerouteUsage()), False=@n88) {group=*S26_SL, parameterName=UseT3}
  n90 = NamedRerouteUsage()
  n91 = NamedRerouteUsage()
  n92 = NamedRerouteUsage()
  n93 = NamedRerouteUsage()
  n94 = LinearInterpolate(
   A=StaticSwitchParameter(True=LinearInterpolate(A=@n89, B=@n90, Alpha=@n91), False=LinearInterpolate(A=@n89, B=@n92, Alpha=@n93) {constB=0}) {group=*S26_SL_Primary_Lazarus, parameterName=isLazarusLensStyleMask}
   B=Multiply(A=NamedRerouteUsage(), B=ScalarParameter() {group=*S26SL_Tertiary, parameterName=Tertiary_EmissiveIntensity})
   Alpha=NamedRerouteUsage()
  )
  n95 = StaticSwitchParameter(True=LinearInterpolate(A=@n94, B=NamedRerouteUsage(), Alpha=NamedRerouteUsage()) {constA=0.5, constB=0}, False=@n94) {group=*S26_SL_Hair, parameterName=UseHair}
  n96 = LinearInterpolate(
   A=Desaturation(None=NamedRerouteUsage())
   B=Constant3Vector() {constant=<rgba 1 1 1 1>}
   Alpha=ScalarParameter() {defaultValue=0.2, group=*S26_SL_Quaternary, parameterName=FlatBaseColor_Intensity}
  )
  n97 = NamedRerouteUsage()
  n98 = StaticSwitchParameter(
   True=LinearInterpolate(A=@n95, B=Multiply(A=@n96, B=@n97), Alpha=NamedRerouteUsage())
   False=LinearInterpolate(A=@n95, B=NamedRerouteUsage(), Alpha=NamedRerouteUsage())
  ) {group=*S26_SL, parameterName=isToon}
  n99 = StaticSwitchParameter(
   True=Add(
    A=MaterialFunctionCall(TattooMask=@n87) {materialFunction=&MF_LazarusLens_TattooGlow}.Result
    B=@n98
   )
   False=@n98
  ) {group=*S26_SL, parameterName=isLazarusWingHead}
  NamedRerouteDeclaration(None=StaticSwitchParameter(True=Add(A=@n86.Result, B=@n99), False=@n99) {group=*S26_SL_Primary_Lazarus, parameterName=UseLazarusTatooBodyEmissive}) {name=Emissive_OUT, nodeColor=<rgba 1 0 0.5615 1>}
  NamedRerouteDeclaration(None=@n28.Emissive Color) {name=Emissive_IN, nodeColor=<rgba 0.8984 0 1 1>}
  NamedRerouteDeclaration(None=@n28.Base Color) {name=BaseColor_IN, nodeColor=<rgba 0 1 0.0156 1>}
  n100 = TextureObjectParameter() {
   group=*S26SL_Primary
   parameterName=PrimaryPattern
   samplerType=SAMPLERTYPE_LinearColor
   texture=/Game/Characters/Player/Female/Medium/Bodies/F_MED_PizzaParty/Skins/SuperLevel/Textures/T_T3MatrixLines
  }
  NamedRerouteDeclaration(None=@n100) {name=T_PrimaryPattern, nodeColor=<rgba 0 0.4844 1 1>}
  NamedRerouteDeclaration(None=@n84.Waves2) {name=LazarusLensMask_Waves2, nodeColor=<rgba 0 1 0.7188 1>}
  n101 = Frac(None=Multiply(A=NamedRerouteUsage()) {constA=20, constB=32})
  n102 = Constant() {r=0.5}
  n103 = OneMinus(None=Multiply(A=Distance(A=@n101, B=@n102)) {constB=2})
  n104 = NamedRerouteUsage()
  n105 = AppendVector(
   A=ComponentMask(None=@n104) {r=true}
   B=Divide(A=ComponentMask(None=@n104) {g=true}, B=ScalarParameter() {defaultValue=1, group=*S26SL_Primary_T3, parameterName=T3_MatrixLength})
  )
  n106 = TextureSample(UVs=Panner(Coordinate=@n105, Time=NamedRerouteUsage(), Speed=NamedRerouteUsage()), Tex=NamedRerouteUsage()) {
   samplerType=SAMPLERTYPE_LinearColor
   texture=/Game/Characters/Player/Female/Medium/Bodies/F_MED_PizzaParty/Skins/SuperLevel/Textures/T_ArtDeco_T1_Masks
  }
  n107 = NamedRerouteUsage()
  n108 = ScalarParameter() {defaultValue=0.25, group=*S26SL_Primary_T3, parameterName=T3_SecondaryScrollMultiplier}
  n109 = TextureSample(
   UVs=Panner(Coordinate=Add(A=@n105) {constB=0.25}, Time=NamedRerouteUsage(), Speed=Multiply(A=@n107, B=@n108) {constB=0.25})
   Tex=NamedRerouteUsage()
  ) {
   samplerType=SAMPLERTYPE_LinearColor
   texture=&T_ArtDeco_T1_Masks
  }
  n110 = Add(A=@n106.R, B=@n109.R)
  n111 = LinearInterpolate(B=NamedRerouteUsage(), Alpha=Multiply(A=@n110) {constB=0.5})
  NamedRerouteDeclaration(None=Step(X=Multiply(A=@n103, B=@n111)) {constX=0.5, constY=0.03}) {name=T3_Matrix_Lines, nodeColor=<rgba 0.8984 1 0 1>}
  NamedRerouteDeclaration(None=Constant()) {name=Hair_Roughness, nodeColor=<rgba 0 1 0.3438 1>}
  NamedRerouteDeclaration(None=@n83.TatooBaseColor) {name=LazarusWing_FaceBaseColor, nodeColor=<rgba 0.0078 0 1 1>}
  n112 = ScalarParameter() {defaultValue=0.2, group=*S26_SL_Primary_Lazarus, parameterName=Lines2_PrimaryColorMultiply}
  n113 = Saturate(None=Add(A=NamedRerouteUsage(), B=NamedRerouteUsage()))
  n114 = LinearInterpolate(
   A=LinearInterpolate(A=@MaterialExpressionReroute_1, B=Multiply(A=@n18, B=@n112) {constB=0.1}, Alpha=NamedRerouteUsage())
   B=NamedRerouteUsage()
   Alpha=@n113
  )
  NamedRerouteDeclaration(None=LinearInterpolate(
    A=NamedRerouteDeclaration(None=@n114) {name=LazarusMaskLines2_BaseColor, nodeColor=<rgba 0 1 0.6484 1>}
    B=NamedRerouteUsage()
    Alpha=NamedRerouteUsage()
   )) {name=LazarusStyle_BaseColor, nodeColor=<rgba 0 0.9297 1 1>}
  NamedRerouteDeclaration(None=@n84.WavesMask) {name=LazarusLens_MaskLines, nodeColor=<rgba 1 0.5859 0 1>}
  n115 = ScalarParameter() {group=*S26_SL_Primary_Lazarus, parameterName=LazarusStyleMask_RoughnessBase, sortPriority=1}
  n116 = LinearInterpolate(
   A=LinearInterpolate(
    B=ScalarParameter() {group=*S26_SL_Primary_Lazarus, parameterName=LazarusStyleMask_RoughnessLines1, sortPriority=2}
    Alpha=NamedRerouteUsage()
   )
   B=ScalarParameter() {group=*S26_SL_Primary_Lazarus, parameterName=LazarusStyleMask_RoughnessLines2, sortPriority=2}
   Alpha=@n113
  )
  n117 = Saturate(None=Add(A=NamedRerouteUsage(), B=NamedRerouteUsage()))
  NamedRerouteDeclaration(None=LinearInterpolate(A=LinearInterpolate(A=@n115, B=@n116, Alpha=@n117), B=NamedRerouteUsage(), Alpha=NamedRerouteUsage())) {name=LazarusLensMask_Roughness, nodeColor=<rgba 1 0 0.9375 1>}
  n118 = ScalarParameter() {group=*S26_SL_Primary_Lazarus, parameterName=LazarusStyleMask_MetallicBase, sortPriority=1}
  n119 = ScalarParameter() {group=*S26_SL_Primary_Lazarus, parameterName=LazarusStyleMask_MetallicLines, sortPriority=2}
  NamedRerouteDeclaration(None=LinearInterpolate(A=LinearInterpolate(A=@n118, B=@n119, Alpha=@n117), B=NamedRerouteUsage(), Alpha=NamedRerouteUsage())) {name=LazarusLensMask_Metallic, nodeColor=<rgba 0 1 0.9531 1>}
  NamedRerouteDeclaration(None=ScalarParameter() {defaultValue=0.35, group=*S26SL_Tertiary, parameterName=Tertiary_Specular}) {name=Tertiary_Specular, nodeColor=<rgba 0 0.2031 1 1>}
  n120 = VectorParameter() {group=*S26SL_Primary, parameterName=PrimaryColor, sortPriority=0, defaultValue=<rgba 0.05 0.05 0.05 1>}
  NamedRerouteDeclaration(None=@n120.RGB) {name=PrimaryColor, nodeColor=<rgba 0.3125 0 1 1>}
  NamedRerouteDeclaration(None=StaticSwitchParameter(True=ScalarParameter() {defaultValue=1, group=*S26SL_Secondary, parameterName=Secondary_Specular}, False=@n45.Specular) {group=*S26_SL, parameterName=UseT3}) {name=Secondary_Specular, nodeColor=<rgba 0 0.5781 1 1>}
  n121 = ScalarParameter() {group=*S26_SL_Primary_Lazarus, parameterName=LazarusStyleMask_SpecularBase, sortPriority=1}
  n122 = ScalarParameter() {group=*S26_SL_Primary_Lazarus, parameterName=LazarusStyleMask_SpecularLines, sortPriority=2}
  NamedRerouteDeclaration(None=LinearInterpolate(A=LinearInterpolate(A=@n121, B=@n122, Alpha=@n117), B=NamedRerouteUsage(), Alpha=NamedRerouteUsage())) {name=LazarusLensMask_Specular, nodeColor=<rgba 0 0.6953 1 1>}
  n123 = NamedRerouteUsage()
  n124 = Add(A=NamedRerouteUsage(), B=NamedRerouteUsage())
  n125 = Multiply(
   A=LinearInterpolate(B=NamedRerouteUsage(), Alpha=Saturate(None=@n124))
   B=ScalarParameter() {defaultValue=0.2, group=*S26_SL_Primary_Lazarus, parameterName=LazarusMaskLinesEmissiveIntensity}
  )
  NamedRerouteDeclaration(None=LinearInterpolate(A=Add(A=@n123, B=@n125), B=NamedRerouteUsage(), Alpha=NamedRerouteUsage())) {name=LazarusStyleMask_Emissive, nodeColor=<rgba 0 0.0391 1 1>}
  NamedRerouteDeclaration(None=@n28.Normal) {name=Normal_IN, nodeColor=<rgba 0.6798 0.1602 1 1>}
  NamedRerouteDeclaration(None=@n20.Tech Art Mask) {name=TechArtMask_In, nodeColor=<rgba 1 0.0234 0 1>}
  NamedRerouteDeclaration(None=Constant()) {name=Hair_Specular, nodeColor=<rgba 0.9219 0 1 1>}
  n126 = Multiply(A=NamedRerouteUsage(), B=ScalarParameter() {defaultValue=1, group=*S26SL_Primary_T2, parameterName=PerlinDistortion})
  n127 = NamedRerouteUsage()
  n128 = TextureSample(UVs=Panner(Coordinate=Add(A=@n126, B=@n127), Time=NamedRerouteUsage(), Speed=NamedRerouteUsage()), Tex=NamedRerouteUsage()) {
   samplerType=SAMPLERTYPE_LinearColor
   texture=&T_ArtDeco_T1_Masks
  }
  n129 = NamedRerouteUsage()
  NamedRerouteDeclaration(None=MaterialFunctionCall!(
    SDF=Multiply(A=@n128.R, B=@n129)
    TimeOffset
    LineThickness
    LineDistance=ScalarParameter() {defaultValue=1, group=*S26SL_Primary_T2, parameterName=ShapeLineDistance, sortPriority=50}
    VariableThickness=StaticBoolParameter() {defaultValue=true, group=*S26SL_Primary_T2, parameterName=ShapeLineVariableThickness, sortPriority=51}
    ThicknessRangeMin
    ThicknessRangeMax
    ThicknessMin
    ThicknessMax=ScalarParameter() {defaultValue=0.35, group=*S26SL_Primary_T2, parameterName=ShapeLineWidthMax, sortPriority=55}
    Animate=StaticBoolParameter() {defaultValue=true, group=*S26SL_Primary_T3, parameterName=ShapeLineAnimate, sortPriority=56}
    Speed=ScalarParameter() {defaultValue=0.15, group=*S26SL_Primary_T2, parameterName=ShapeLineSpeed, sortPriority=57}
    Pulse=StaticBoolParameter() {group=*S26SL_Primary_T3, parameterName=ShapeLinePulse, sortPriority=58}
    PulseFrequency
    PulseSpeed
   ) {materialFunction=/Game/Global/MaterialFunctions/MF_SDF_Lines}.Lines) {name=T2_SDF_Lines, nodeColor=<rgba 0 1 0.3906 1>}
  n130 = MaterialFunctionCall(Float3=@n20.M) {materialFunction=&BreakOutFloat3Components}
  NamedRerouteDeclaration(None=@n130.G) {name=Gmap, nodeColor=<rgba 0 0.5547 1 1>}
  n131 = Desaturation(None=NamedRerouteUsage())
  n132 = Constant3Vector() {constant=<rgba 1 1 1 1>}
  n133 = ScalarParameter() {defaultValue=0.2, group=*S26_SL_Quaternary, parameterName=FlatBaseColor_Intensity}
  NamedRerouteDeclaration(None=Multiply(A=LinearInterpolate(A=@n131, B=@n132, Alpha=@n133), B=NamedRerouteUsage())) {name=Quaternary_BaseColor, nodeColor=<rgba 0.7578 0 1 1>}
  n134 = NamedRerouteUsage()
  n135 = StaticSwitchParameter(True=LinearInterpolate(A=@n134, B=NamedRerouteUsage(), Alpha=NamedRerouteUsage()), False=@n134) {defaultValue=true, parameterName=UseT1}
  n136 = NamedRerouteUsage()
  n137 = NamedRerouteUsage()
  NamedRerouteDeclaration(None=StaticSwitchParameter(True=LinearInterpolate(A=@n135, B=@n136, Alpha=@n137) {constA=0.5, constB=0}, False=@n135) {group=*S26_SL_Hair, parameterName=UseHair}) {name=Normal_OUT, nodeColor=<rgba 1 0 0.1406 1>}
  n138 = ChannelMaskParameter(None=@n48) {group=*S26_SL, maskChannel=Alpha, parameterName=EmissiveChannel, sortPriority=5, defaultValue=<rgba 0 0 0 1>}
  NamedRerouteDeclaration(None=Multiply(A=@n138.R, B=ScalarParameter() {defaultValue=1, group=*S26_SL, parameterName=EmissiveChannel_Intensity})) {name=Emissive_Mask, nodeColor=<rgba 0.5729 0.5729 0.5729 1>}
  NamedRerouteDeclaration(None=VectorParameter() {group=*S26_SL, parameterName=EmissiveColor, sortPriority=0, defaultValue=<rgba 0 0 0 1>}.RGB) {name=EmissiveColor, nodeColor=<rgba 0.3348 0 0.3542 1>}
  n139 = TextureObjectParameter() {
   group=*S26_SL_T1 Pattern
   parameterName=T1_PrimaryPatternMask
   samplerType=SAMPLERTYPE_Masks
   texture=&T_BlackMask
  }
  NamedRerouteDeclaration(None=@n139) {name=T_T1PrimaryPattern_Mask, nodeColor=<rgba 1 0 0.4922 1>}
  NamedRerouteDeclaration(None=Multiply(A=NamedRerouteUsage(), B=ScalarParameter() {defaultValue=1, group=*S26_SL_Quaternary, parameterName=Quaternary_EmissiveIntensity})) {name=Quaternary_Emissive, nodeColor=<rgba 1 0 0.8672 1>}
  n140 = VectorParameter() {group=*S26_SL_Quaternary, parameterName=Quaternary_BaseColor_Multiply, defaultValue=<rgba 0.05 0.05 0.05 1>}
  NamedRerouteDeclaration(None=@n140.RGB) {name=QuaternaryColor, nodeColor=<rgba 0 1 0.2969 1>}
  NamedRerouteDeclaration(None=@n130.B) {name=Skin, nodeColor=<rgba 0 1 0.3438 1>}
  n141 = MaterialFunctionCall!(
   Position=Add(
    A=MaterialFunctionCall() {materialFunction=&ObjectPivotPoint}.Object Pivot Location
    B=Constant3Vector() {constant=<rgba 0 0 100 0>}
   )
   UVRadialScale
  ) {materialFunction=/Game/Characters/Player/Common/MaterialFunctions/Superlevel/CharacterStabilizedScreenSpaceUVs}
  n142 = ScalarParameter() {defaultValue=5, group=*S26SL_Primary, parameterName=PatternSize}
  NamedRerouteDeclaration(None=MaterialFunctionCall(
    UVs=Multiply(A=@n141.StabilizedUVs, B=@n142) {constB=3}
    "Rotation Angle (0-1)"=ScalarParameter() {group=*S26SL_Primary, parameterName=PatternAngle}
   ) {materialFunction=&CustomRotator}.Rotated Values) {name=PrimaryScreenUV, nodeColor=<rgba 0 0.5312 1 1>}
  n143 = NamedRerouteUsage()
  MaterialExpressionReroute_10 = Reroute(None=@MaterialExpressionReroute_3.mask)
  MaterialExpressionReroute_11 = Reroute(None=@MaterialExpressionReroute_10.pattern)
  MaterialExpressionReroute_12 = Reroute(None=@MaterialExpressionReroute_11.time)
  MaterialExpressionReroute_13 = Reroute(None=@MaterialExpressionReroute_12.mask)
  MaterialExpressionReroute_14 = Reroute(None=@MaterialExpressionReroute_3)
  MaterialExpressionReroute_16 = Reroute(None=@MaterialExpressionReroute_8)
  n144 = Add(A=Multiply(A=@MaterialExpressionReroute_14.lineThickness) {constB=-1}, B=@MaterialExpressionReroute_16.time)
  MaterialExpressionReroute_15 = Reroute(None=@n10.B)
  MaterialExpressionReroute_17 = Reroute(None=@n12)
  n145 = Multiply(
   A=Subtract(
    A=Step(Y=@n144, X=@MaterialExpressionReroute_15.pattern)
    B=Step(Y=@MaterialExpressionReroute_16.time, X=@MaterialExpressionReroute_15.pattern)
   )
   B=@MaterialExpressionReroute_17.mask
  )
  MaterialExpressionReroute_18 = Reroute(None=@n145)
  n146 = Saturate(None=Add(A=Add(A=@MaterialExpressionReroute_4.Output, B=@MaterialExpressionReroute_13.Output), B=@MaterialExpressionReroute_18.Output))
  NamedRerouteDeclaration(None=StaticSwitchParameter(True=NamedRerouteUsage(), False=StaticSwitchParameter(True=@n143, False=@n146) {group=*S26_SL, parameterName=UseT2}) {group=*S26_SL, parameterName=UseT3}) {name=PrimaryPattern, nodeColor=<rgba 0 1 0.9062 1>}
  NamedRerouteDeclaration(None=Multiply(A=Time(), B=ScalarParameter() {defaultValue=0.2, group=*S26SL_Primary, parameterName=Anim_PatternScrollSpeed})) {name=PrimaryScrollMultiplier, nodeColor=<rgba 0 1 0.7422 1>}
  n147 = NamedRerouteDeclaration(None=@n110) {name=T3_PrimaryColorGradient, nodeColor=<rgba 1 0 0.0326 1>}
  n148 = Multiply(A=@n147, B=ScalarParameter() {defaultValue=5, group=*S26SL_Primary_T3, parameterName=T3_ColorGradient})
  NamedRerouteDeclaration(None=LinearInterpolate(A=NamedRerouteUsage(), B=NamedRerouteUsage(), Alpha=Multiply(A=@n147, B=@n148))) {name=T3_PrimaryInnerColor, nodeColor=<rgba 0.9688 0 1 1>}
  NamedRerouteDeclaration(None=AppendVector(
    A=ScalarParameter() {group=*S26SL_Primary, parameterName=Layer1SpeedX}
    B=ScalarParameter() {defaultValue=-0.5, group=*S26SL_Primary, parameterName=Layer1SpeedY}
   )) {name=PrimaryScrollSpeed, nodeColor=<rgba 0.7109 1 0 1>}
  n149 = Multiply(
   A=Multiply(A=NamedRerouteUsage(), B=NamedRerouteUsage())
   B=ScalarParameter() {defaultValue=0.5, group=*S26SL_Primary, parameterName=PrimaryPatternEmissiveIntensity}
  )
  n150 = Add(A=ScalarParameter() {defaultValue=5, group=*S26SL_Primary, parameterName=Primary_FresnelHardness})
  n151 = Multiply(A=NamedRerouteUsage(), B=@n150)
  n152 = OneMinus(None=Add(A=@n150) {constB=-1})
  n153 = NamedRerouteDeclaration(None=Saturate(None=Add(A=@n151, B=@n152) {constB=0})) {name=EmissiveFresnel_1, nodeColor=<rgba 0.2187 1 0 1>}
  n154 = NamedRerouteUsage()
  n155 = LinearInterpolate(
   B=Desaturation(None=NamedRerouteUsage(), Fraction=ScalarParameter() {defaultValue=-1, group=*S26SL_Primary, parameterName=EmissiveDesat})
   Alpha=Multiply(A=Multiply(A=@n153, B=@n154), B=ScalarParameter() {defaultValue=1, group=*S26SL_Primary, parameterName=PrimaryFresnelEmissive_Intensity})
  )
  NamedRerouteDeclaration(None=StaticSwitchParameter(True=@n149, False=Add(A=@n149, B=@n155)) {group=*S26_SL, parameterName=UseT3}) {name=Primary_Emissive, nodeColor=<rgba 0 0.0156 1 1>}
  n156 = Add(
   A=Multiply(A=@n141.StabilizedUVs, B=ScalarParameter() {defaultValue=20, group=*S26SL_Primary, parameterName=Primary_PerlinScale}) {constB=3}
   B=Multiply(A=Time(), B=ScalarParameter() {defaultValue=0.5, group=*S26SL_Primary, parameterName=Primary_Perlin_Speed})
  )
  n157 = NamedRerouteUsage()
  NamedRerouteDeclaration(None=Multiply(
    A=TextureSample(UVs=@n156, Tex=@n157).R
    B=ScalarParameter() {defaultValue=2.5, group=*S26SL_Primary, parameterName=Primary_PerlinScreenspace_Multiplier}
   )) {name=Perlin_Screenspace, nodeColor=<rgba 1 0 0.2578 1>}
  n158 = TextureObject() {
   samplerType=SAMPLERTYPE_LinearColor
   texture=/Game/Characters/Player/Female/Medium/Bodies/F_MED_PizzaParty/Skins/SuperLevel/Textures/T_SE26_SL_Perlin
  }
  NamedRerouteDeclaration(None=@n158) {name=T_Perlin, nodeColor=<rgba 0 0.0859 1 1>}
  n159 = MaterialFunctionCall(Color=@n47) {materialFunction=&SplitComponents}
  n160 = Min(A=@n159.G, B=Min(A=@n159.R, B=@n159.R))
  n161 = OneMinus(None=Max(A=@n159.G, B=Max(A=@n159.R, B=@n159.R)))
  NamedRerouteDeclaration(None=Max(
    A=NamedRerouteDeclaration(None=@n160) {name=Skin_Mask, nodeColor=<rgba 1 1 1 1>}
    B=NamedRerouteDeclaration(None=@n161) {name=Quatenary_Mask, nodeColor=<rgba 0 0 0 1>}
   )) {name=NotSL_Mask, nodeColor=<rgba 0.3125 1 0 1>}
  n162 = VectorParameter() {group=*S26_SL_Hair, parameterName=Hair_Color1, defaultValue=<rgba 0.0533 0.04 0.2 1>}
  n163 = VectorParameter() {group=*S26_SL_Hair, parameterName=Hair_Color2, defaultValue=<rgba 0.169 0.3575 1.3 1>}
  n164 = NamedRerouteUsage()
  n165 = NamedRerouteUsage()
  n166 = ScalarParameter() {defaultValue=1, group=*S26_SL_Hair, parameterName=GmapInfluence}
  n167 = LinearInterpolate(A=@n164, B=@n165, Alpha=Multiply(A=@n164, B=Multiply(A=@n165, B=@n166))) {constAlpha=1}
  n168 = SmoothStep(Max=ScalarParameter() {defaultValue=0.2, group=*S26_SL_Hair, parameterName=HairColorTransition1}, Value=@n167) {constMax=0.2}
  n169 = VectorParameter() {group=*S26_SL_Hair, parameterName=Hair_Color3, defaultValue=<rgba 0.1 2 1.2893 1>}
  n170 = ScalarParameter() {defaultValue=0.5, group=*S26_SL_Hair, parameterName=HairColorTransition2}
  n171 = NamedRerouteUsage()
  n172 = NamedRerouteUsage()
  n173 = Desaturation(None=GetMaterialAttributes(None=MaterialFunctionCall!(In=@n171, Warp Mask, GalaxyAlternateFresnel, Alpha=@n172) {materialFunction=/Game/Weapons/WeaponSkins/MaterialFunction/Galaxy_SL26_hair}.Result) {attributeGetTypes [#1]}.Emissive Color)
  n174 = Add(
   A=LinearInterpolate(A=LinearInterpolate(A=@n162.RGB, B=@n163.RGB, Alpha=@n168), B=@n169.RGB, Alpha=SmoothStep(Min=@n170, Value=@n167) {constMin=0.2})
   B=Multiply(A=@n169.RGB, B=Multiply(A=@n173, B=@n173))
  )
  n175 = LinearInterpolate(Alpha=NamedRerouteUsage()) {constB=1.2}
  NamedRerouteDeclaration(None=StaticSwitchParameter(True=@n174, False=Multiply(A=@n174, B=@n175)) {group=*S26_SL, parameterName=isToon}) {name=Hair_Emissive, nodeColor=<rgba 0 0.5547 1 1>}
  NamedRerouteDeclaration(None=Multiply(A=NamedRerouteUsage(), B=ScalarParameter() {defaultValue=1, group=*S26_SL_Quaternary, parameterName=Quaternary_SpecularIntensity})) {name=Quaternary_Specular, nodeColor=<rgba 0.5703 0 1 1>}
  PinBase() {reroutePins [
    lineThickness {expression=&self:MaterialGraph_0.MaterialGraphNode_Composite_3.Collapsed Nodes.MaterialExpressionReroute_0}
    pattern {expression=&self:MaterialGraph_0.MaterialGraphNode_Composite_3.Collapsed Nodes.MaterialExpressionReroute_1}
    time {expression=&self:MaterialGraph_0.MaterialGraphNode_Composite_3.Collapsed Nodes.MaterialExpressionReroute_2}
    mask {expression=&self:MaterialGraph_0.MaterialGraphNode_Composite_3.Collapsed Nodes.MaterialExpressionReroute_3}
    linesAmount {expression=&self:MaterialExpressionReroute_2}
   ]}
  PinBase!(Output=@n145) {reroutePins [Output {expression=&self:MaterialExpressionReroute_18}]}
  n176 = Frac(None=Multiply(A=@MaterialExpressionReroute_1.pattern, B=@MaterialExpressionReroute_2.linesAmount))
  n177 = Step(Y=Add(A=Multiply(A=@MaterialExpressionReroute_0.lineThickness) {constB=-1}, B=@MaterialExpressionReroute_2.time), X=@n176)
  n178 = Step(Y=@MaterialExpressionReroute_2.time, X=@n176)
  PinBase(Output=Multiply(A=Subtract(A=@n177, B=@n178), B=@MaterialExpressionReroute_10)) {reroutePins [Output {expression=&self:MaterialGraph_0.MaterialGraphNode_Composite_3.Collapsed Nodes.MaterialExpressionReroute_4}]}
  PinBase!() {reroutePins [
    lineThickness {expression=&self:MaterialGraph_0.MaterialGraphNode_Composite_4.Collapsed Nodes.MaterialExpressionReroute_8}
    pattern {expression=&self:MaterialGraph_0.MaterialGraphNode_Composite_4.Collapsed Nodes.MaterialExpressionReroute_10}
    time {expression=&self:MaterialGraph_0.MaterialGraphNode_Composite_4.Collapsed Nodes.MaterialExpressionReroute_11}
    mask {expression=&self:MaterialGraph_0.MaterialGraphNode_Composite_4.Collapsed Nodes.MaterialExpressionReroute_12}
    linesAmount {expression=&self:MaterialExpressionReroute_6}
   ]}
  n179 = Step(
   Y=Add(A=Multiply(A=@MaterialExpressionReroute_8.lineThickness) {constB=-1}, B=@MaterialExpressionReroute_12)
   X=@MaterialExpressionReroute_11
  )
  n180 = Step(Y=@MaterialExpressionReroute_12, X=@MaterialExpressionReroute_11)
  PinBase!(Output=Multiply(A=Subtract(A=@n179, B=@n180), B=@MaterialExpressionReroute_13)) {reroutePins [Output {expression=&self:MaterialGraph_0.MaterialGraphNode_Composite_4.Collapsed Nodes.MaterialExpressionReroute_13}]}
  PinBase!() {reroutePins [
    lineThickness {expression=&self:MaterialExpressionReroute_14}
    pattern {expression=&self:MaterialExpressionReroute_15}
    time {expression=&self:MaterialExpressionReroute_16}
    mask {expression=&self:MaterialExpressionReroute_17}
    linesAmount {expression=&self:MaterialExpressionReroute_9}
   ]}
  MaterialExpressionReroute_6 = Reroute(None=@n8)
  MaterialExpressionReroute_9 = Reroute(None=@n13)
  Step(Y=@MaterialExpressionReroute_1.pattern, X=Clamp(None=@MaterialExpressionReroute_1.pattern) {maxDefault=0.4, minDefault=0.3})
  n181 = Multiply(A=@MaterialExpressionReroute_0.lineThickness) {constB=-1}
  n182 = Sine(None=@n19)
  Subtract(A=Step(Y=Add(A=@n181, B=@MaterialExpressionReroute_2.time), X=@n182), B=Step(Y=@MaterialExpressionReroute_2.time, X=@n182))
 ]
}
